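; Anim notify that controls the lightcone Niagara effect on the Augment Selector Tool weapon during animations.
; On notify, it casts the mesh owner to a Fort player pawn, then casts that pawn's current weapon to the selector tool.
; If valid, it pushes its own toggle, instant-off, flash and flicker booleans into the lightcone component's Niagara variables.
; It returns the parent result on success, otherwise false.
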(fn Received_Notify
  (params (MeshComp :type Skeletal%20Mesh%20Component%20Object%20Reference) (Animation :type Anim%20Sequence%20Base%20Object%20Reference) (EventReference :type Anim%20Notify%20Event%20Reference%20Structure%20%28by%20ref%29))
  (bind _returnvalue (|CallParentFunction MeshComp Animation EventReference))
  (bind _asfort_player_pawn (Utilities|Casting|CastToFortPlayerPawn (Components|GetOwner MeshComp)))
  (bind _asaugment_selector_tool_weapon (Utilities|Casting|CastToAugmentSelectorToolWeapon (Weapon|GetCurrentWeapon _asfort_player_pawn))
    (:then
      (Utilities|IsValid _asaugment_selector_tool_weapon
        (:Is%20Valid
          (Niagara|SetNiagaraVariable(Bool) (Variables|/PlayerAugments/SelectionTool/AugmentSelectorToolWeapon.AugmentSelectorToolWeapon_C|GetNS_Aug_Granter_Lightcone _asaugment_selector_tool_weapon) "LightConeToggle" (Variables|Default|GetLightConeToggle))
          (Niagara|SetNiagaraVariable(Bool) (Variables|/PlayerAugments/SelectionTool/AugmentSelectorToolWeapon.AugmentSelectorToolWeapon_C|GetNS_Aug_Granter_Lightcone _asaugment_selector_tool_weapon) "LightConeInstantOff" (Variables|Default|GetLightConeInstantOff))
          (Niagara|SetNiagaraVariable(Bool) (Variables|/PlayerAugments/SelectionTool/AugmentSelectorToolWeapon.AugmentSelectorToolWeapon_C|GetNS_Aug_Granter_Lightcone _asaugment_selector_tool_weapon) "LightConeFlash" (Variables|Default|GetLightConeFlash))
          (Niagara|SetNiagaraVariable(Bool) (Variables|/PlayerAugments/SelectionTool/AugmentSelectorToolWeapon.AugmentSelectorToolWeapon_C|GetNS_Aug_Granter_Lightcone _asaugment_selector_tool_weapon) "LightConeFlicker" (Variables|Default|GetLightConeFlicker))
          (return _returnvalue))
        (:Is%20Not%20Valid
          _)))
    (:CastFailed
      _))
  (return false))
