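# Sound cue named for the looping hologram audio of the Augment granter in the PlayerAugments selection tool.
# Its graph shows no root or sound nodes, so only cue-level settings are defined here.
# It routes to the Fort_Gadgets_LTM sound class with boosted volume and uses inline attenuation and concurrency overrides, giving a long falloff and limiting playback to one instance.
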SoundCue /PlayerAugments/SelectionTool/Assets/Sound/Aug_Granter_Hologram_Loop_Cue
root=null
nodes []
properties {
 bOverrideAttenuation=true
 bOverrideConcurrency=true
 maxDistance=5400
 soundClassObject=/Game/Sounds/FortSoundClassesAndModes/Fort_Gadgets_LTM
 volumeMultiplier=2
 attenuationOverrides {falloffDistance=5000}
 concurrencyOverrides {maxCount=1}
}
